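# Lit material for a tire pile that plays a baked rigid-piece animation and adds highlight effects.
# Vertex red selects rows in an animation texture, whose position, quaternion and optional scale drive World Position Offset around first-frame pivots, with the quaternion also rotating normals.
# Frames follow time since GameTime or a preview override, clamped between start and end.
# Switchable emissive adds a Fresnel glow and a camera-oriented moving sweep, phase-offset by object position.
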Material /ValetMods/Meshes/Tires/Tire_Pile/Materials/M_Tire_pile
graph {
 properties {bTangentSpaceNormal=false, bUseMaterialAttributes=true, pixelDepthOffsetMode=PDOM_Legacy, usedShadingModels=Default Lit}
 nodes [
  n1 = TextureSample() {samplerType=SAMPLERTYPE_Masks, texture=/ValetMods/Meshes/Tires/OffRoad/Textures/T_Off_Road_Tires_S}
  n2 = StaticSwitchParameter(True=Constant() {r=0.3333}, False=Constant() {r=0.5}) {parameterName=ConstainsScaleInfo}
  n3 = Multiply(
   A=ScalarParameter() {defaultValue=7, group=BakedTransforms, parameterName=numSockets, sortPriority=5}
   B=StaticSwitchParameter(True=Constant() {r=3}, False=Constant() {r=2}) {desc=Whether the baked transformation texture includes scale information, parameterName=ConstainsScaleInfo}
  ) {constB=3}
  n4 = Add(
   A=Multiply(A=VertexColor().R, B=@n2) {constB=0.5}
   B=Multiply(A=Divide(B=@n3) {constB=23}, B=ScalarParameter() {defaultValue=0.5, group=BakedTransforms, parameterName=SubPixelMult}) {constB=0.25}
  ) {constB=-0.1}
  n5 = ScalarParameter() {group=Animation, parameterName=StartFrame, sortPriority=30}
  n6 = ScalarParameter() {defaultValue=30, parameterName=AnimationFPS}
  n7 = Subtract(A=Time(), B=ScalarParameter() {group=Animation, parameterName=GameTime})
  n8 = StaticSwitchParameter(
   True=ScalarParameter() {defaultValue=30, group=Deformation, parameterName=PreviewFrame, sortPriority=80}
   False=Multiply(A=Multiply(A=@n6, B=@n7), B=ScalarParameter() {defaultValue=1, parameterName=AnimationSpeedMultiplier})
  ) {defaultValue=true, group=Deformation, parameterName=Override Frame}
  n9 = ScalarParameter() {
   defaultValue=54
   desc="The number of pixels on the X axis, or the number of frames in the original animation"
   group=Animation
   parameterName=AnimationSequenceLength
   sortPriority=20
  }
  n10 = Divide(B=@n9) {constB=40}
  n11 = Clamp(
   None=Multiply(A=Add(A=@n8, B=@n5) {desc=Offset animation when start isn't 0}, B=@n10) {constB=1.5}
   Min=Divide(A=@n5, B=@n9)
   Max=Divide(A=ScalarParameter() {defaultValue=20, group=Animation, parameterName=EndFrame, sortPriority=40}, B=@n9)
  )
  n12 = TextureSampleParameter2D(UVs=MaterialFunctionCall(X=@n11, Y=@n4) {materialFunction=&MakeFloat2}.Result) {
   desc=position
   group=BakedTransforms
   parameterName=AnimationTexture
   samplerType=SAMPLERTYPE_LinearColor
   sortPriority=0
   texture=/ValetMods/Meshes/Tires/Tire_Pile/Textures/T_Tire_Pile_Animation
  }
  n13 = Add(A=@n4, B=@n2) {constB=0.5}
  n14 = TextureSampleParameter2D(UVs=MaterialFunctionCall(X=@n11, Y=Add(A=@n13, B=@n2)) {materialFunction=&MakeFloat2}.Result) {
   desc=position
   group=BakedTransforms
   parameterName=AnimationTexture
   samplerType=SAMPLERTYPE_LinearColor
   sortPriority=0
   texture=&T_Tire_Pile_Animation
  }
  n15 = PreSkinnedPosition()
  n16 = TextureSampleParameter2D(UVs=MaterialFunctionCall(X=Multiply(A=@n10) {constB=0.5}, Y=@n4) {materialFunction=&MakeFloat2}.Result) {
   desc=position
   group=BakedTransforms
   parameterName=AnimationTexture
   samplerType=SAMPLERTYPE_LinearColor
   sortPriority=0
   texture=&T_Tire_Pile_Animation
  }
  n17 = Subtract(A=@n15, B=@n16.RGB) {constB=0}
  n18 = Multiply(A=@n17, B=@n14.RGB) {constB=0.5}
  n19 = TextureSampleParameter2D(UVs=MaterialFunctionCall(X=@n11, Y=@n13) {materialFunction=&MakeFloat2}.Result) {
   desc=position
   group=BakedTransforms
   parameterName=AnimationTexture
   samplerType=SAMPLERTYPE_LinearColor
   sortPriority=0
   texture=&T_Tire_Pile_Animation
  }
  n20 = MaterialFunctionCall!(
   Vector=StaticSwitchParameter(True=StaticSwitchParameter(True=@n18, False=@n17) {group=Deformation, parameterName=ApplyScaleAnimation, sortPriority=10}, False=@n17) {group=BakedTransforms, parameterName=ConstainsScaleInfo, sortPriority=25}
   Quat=@n19.RGBA
   Invert Quaternion
  ) {materialFunction=/Engine/Functions/Engine_MaterialFunctions03/Math/MultiplyVectorWithQuaternion}
  n21 = Subtract(A=Add(A=Add(A=@n12.RGB, B=@n20.Result) {constB=0}, B=@n16.RGB) {constB=0}, B=@n15)
  n22 = Transform(None=TextureSample() {samplerType=SAMPLERTYPE_Normal, texture=/ValetMods/Meshes/Tires/OffRoad/Textures/T_Off_Road_Tires_N}.RGB) {transformType=TRANSFORM_Local}
  n23 = Multiply(
   A=Fresnel(ExponentIn=ScalarParameter() {defaultValue=6, parameterName=FresnelExponent}) {baseReflectFraction=0}
   B=VectorParameter() {parameterName=GlowColour, defaultValue=<rgba 10 8.7527 2.7576 0>}.RGB
  )
  n24 = Constant()
  n25 = Multiply(A=Subtract(A=WorldPosition().XYZ, B=ObjectPositionWS()), B=ScalarParameter() {defaultValue=0.001, parameterName=WorldTileScale})
  n26 = Normalize(VectorInput=Transform(None=VectorParameter() {parameterName=SweepDirectionCamSpace, defaultValue=<rgba -0.5 0.45 -1 0>}.RGB) {transformSourceType=TRANSFORMSOURCE_Camera})
  n27 = ComponentMask(None=Multiply(A=DotProduct(A=@n25, B=@n26), B=@n26)) {b=true, r=true}
  n28 = ObjectPositionWS()
  n29 = Add(A=Time(), B=Add(A=ComponentMask(None=@n28) {g=true}, B=ComponentMask(None=@n28) {r=true}))
  n30 = Constant() {r=0.28}
  n31 = Frac(None=ComponentMask(None=Panner(Coordinate=@n27, Time=@n29, Speed=@n30) {speedX=0.2, speedY=0.2}) {g=true})
  n32 = Multiply(A=Subtract(A=@n31) {constB=0.9}) {constB=10}
  n33 = Multiply(A=Floor(None=Add(A=@n31) {constB=0.1}), B=Power(Base=OneMinus(None=@n32)) {constExponent=3})
  n34 = Saturate(None=Add(A=MaterialFunctionCall() {materialFunction=&MF_Fresnel_FX}.Fresnel Inverted) {constB=0.1})
  n35 = Multiply(
   A=Fresnel(ExponentIn=ScalarParameter() {defaultValue=6, parameterName=FresnelExponent}) {baseReflectFraction=0.01}
   B=Multiply(A=Multiply(A=@n33, B=@n34), B=VectorParameter() {parameterName=SweepColour, defaultValue=<rgba 10 8.7527 2.7576 0>}.RGB) {constA=1}
  )
  n36 = Constant()
  SetMaterialAttributes!(
   MaterialAttributes
   Base Color=TextureSample() {texture=/ValetMods/Meshes/Tires/OffRoad/Textures/T_Off_Road_Tires_D}.RGB
   Metallic=@n1.G
   Specular=@n1.G
   Roughness=@n1.G
   World Position Offset=StaticSwitchParameter(True=Transform(None=@n21) {transformSourceType=TRANSFORMSOURCE_Local}, False=Constant3Vector()) {defaultValue=true, group=Deformation, parameterName=ApplyWPO, sortPriority=5}
   Normal=Transform(None=MaterialFunctionCall(Vector=@n22, Quat=@n19.RGBA) {materialFunction=&MultiplyVectorWithQuaternion}.Result) {transformSourceType=TRANSFORMSOURCE_Local}
   Emissive Color=Add(
    A=StaticSwitchParameter(True=@n23, False=@n24) {defaultValue=true, parameterName=UseOuterGlow}
    B=StaticSwitchParameter(True=@n35, False=@n36) {defaultValue=true, parameterName=UseSweep}
   )
  ) {attributeSetTypes [
    #1
    #2
    #3
    #4
    #5
    #6
    #7
   ]}
 ]
}
